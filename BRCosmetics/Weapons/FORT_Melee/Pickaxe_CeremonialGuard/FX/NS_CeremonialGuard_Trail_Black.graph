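# Niagara swing trail for the Ceremonial Guard pickaxe, black variant, drawn as one short-lived ribbon emitter.
# AnimTrail modules spawn the ribbon and track it from skeletal mesh bones or sockets over a 0.3-second emitter duration.
# The active initializer uses fixed dark gray; the TrailColor/Glow hue-shift initializer is disabled.
# Particle updates apply warm color and alpha fade curves and feed random values to dynamic material parameters.
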NiagaraSystem /BRCosmetics/Weapons/FORT_Melee/Pickaxe_CeremonialGuard/FX/NS_CeremonialGuard_Trail_Black
dependencies {
 usedDataInterfaces [&NiagaraDataInterfaceCurve, &NiagaraDataInterfaceColorCurve]
 usedDynamicInputs []
 usedModules [
  &SystemState
  &self:SystemSpawnScript.SystemScriptSource.SystemScriptGraph.NiagaraNodeAssignment_0.SetVariables_#1
 ]
}
emitters {AirTrail002 {
  emitterUpdateScript [
   &EmitterState(Life Cycle Mode=::NewEnumerator1, Loop Behavior=::NewEnumerator1, Loop Duration=0.3, UseLoopDelay=false)
   &FN_SpawnAnimTrail_v2(Spawn Interval=0)
  ]
  particleSpawnScript [
   /Game/Effects/Niagara/Modules/AnimTrail/FN_InitializeAnimTrail/FN_InitializeAnimTrailBoneIndecies(Second Bone or Socket Index=1)
   &FN_InitializeAnimTrail(
    AnimTrailWidth=25
    Color=&MultiplyLinearColor(
     Linear Color=&LerpLinearColors(
      StartColor=$User.TrailColor
      EndColor=&HueShiftLinearColor(Hue=-0.05, Saturation=5, Value=10, Linear Color=$User.TrailColor)
      LerpFactor=$User.Glow
     )
     Scale Factor=&MakeFloat4FromVectorAndFloat(
      "Vector (RGB)"=<1 1 1>
      "Float (Alpha)"=&FloatFromCurve(FloatCurve=DataInterface() {propertyValues {Curve {keys [<at 0.07 1>,<at 0.125 0>]}}}, CurveIndex=$Emitter.Age)
     )
    )
    Lifetime=0.15
    SkeletalMesh=$System.Skeletal Mesh
   ) {enabled=false, module=FN_InitializeAnimTrail001}
   &FN_InitializeAnimTrail(Color=<rgba 0.35 0.35 0.35 1>, Lifetime=0.11, SkeletalMesh=$System.Skeletal Mesh)
  ]
  particleUpdateScript [
   &SolveForcesAndVelocity(Clamp Velocity=false, Limit Acceleration=false)
   &FN_UpdateAnimTrail(SkeletalMesh=$System.Skeletal Mesh)
   &ScaleColor(
    Scale Alpha=&FloatFromCurve(FloatCurve=DataInterface() {propertyValues {Curve {keys [<at 0 1 RCIM_Cubic>,<at 1 0>]}}})
    Scale RGB=&MakeVectorFromLinearColorRGB(Color=&ColorFromCurve(ColorCurve=DataInterface() {propertyValues {
        AlphaCurve {keys [<at 0 1 RCIM_Cubic>,<at 1 0 RCIM_Cubic>]}
        BlueCurve {keys [<at 0 0.512 RCIM_Cubic>,<at 0.5 0.4249 RCIM_Cubic>]}
        GreenCurve {keys [<at 0 0.8048 RCIM_Cubic>,<at 0.5 0.59 RCIM_Cubic>]}
        RedCurve {keys [<at 0 1 RCIM_Cubic>,<at 0.5 0.7 RCIM_Cubic>]}
       }}))
    ScaleA=true
    ScaleRGB=true
   )
   &DynamicMaterialParameters(
    Index 0 Param 1=&RandomRangeFloat(Minimum=0, Maximum=1)
    Index 0 Param 2=&RandomRangeFloat(Minimum=0, Maximum=1)
    Index 0 Param 3=1
    Index 0 Param 4=1
    Module.Index0Param0Write=true
    Module.Index0Param1Write=true
    Module.Index0Param2Write=true
    Module.Index0Param3Write=true
   )
  ]
  rendererClasses [&NiagaraRibbonRendererProperties]
  renderers [{rendererClass=&NiagaraRibbonRendererProperties}]
 }}
summary {
 systemName=NS_CeremonialGuard_Trail_Black
 emitters [{emitterName=AirTrail002, rendererClasses [&NiagaraRibbonRendererProperties]}]
 userVariables [
  User.Glow {type=NiagaraFloat, defaultValue=0, description=""}
  User.RibbonIDOffset {type=NiagaraInt32}
  User.SkeletalMesh {
   type=NiagaraDataInterfaceSkeletalMesh
   defaultValue {
    struct=NiagaraExt_VariableValue_DataInterface
    value {dataInterface=&self:NiagaraDataInterfaceSkeletalMesh_0}
   }
  }
  User.TrailColor {type=LinearColor, description="", defaultValue {struct=LinearColor, value=<rgba 0.2893 0.5 0.3596 1>}}
 ]
}
